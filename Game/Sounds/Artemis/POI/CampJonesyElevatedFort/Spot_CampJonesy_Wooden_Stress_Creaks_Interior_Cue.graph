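# Sound cue for intermittent wooden stress creaks as interior spot ambience at the Camp Jonesy elevated fort POI.
# An endless loop runs a pitch-varying modulator into a random delay of several seconds, then an equally weighted pick of five creak waves reused from Reality Bloom WildWest.
# All quality levels share that random node.
# It plays quietly on the indoor ambient sound class with positional one-shot attenuation.
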SoundCue /Game/Sounds/Artemis/POI/CampJonesyElevatedFort/Spot_CampJonesy_Wooden_Stress_Creaks_Interior_Cue
root=Looping_0
nodes [
 {id=Delay_0, children [QualityLevel_0], properties {delayMax=20, delayMin=3}}
 {id=Looping_0, children [Modulator_0]}
 {id=Modulator_0, children [Delay_0], properties {pitchMin=0.85}}
 {id=QualityLevel_0, children [Random_1,Random_1,Random_1,Random_1]}
 {
  id=Random_1
  children [WavePlayer_0,WavePlayer_1,WavePlayer_2,WavePlayer_6,WavePlayer_7]
  properties {bShouldExcludeFromBranchCulling=true, weights [1,1,1,1,1]}
 }
 {id=WavePlayer_0, properties {soundWaveAssetPtr=/Game/Sounds/Artemis/POI/Reality_Bloom_WildWest/Spots/Spot_Wood_Creak_Rnd_01}}
 {id=WavePlayer_1, properties {soundWaveAssetPtr=/Game/Sounds/Artemis/POI/Reality_Bloom_WildWest/Spots/Spot_Wood_Creak_Rnd_02}}
 {id=WavePlayer_2, properties {soundWaveAssetPtr=/Game/Sounds/Artemis/POI/Reality_Bloom_WildWest/Spots/Spot_Wood_Creak_Rnd_03}}
 {id=WavePlayer_6, properties {soundWaveAssetPtr=/Game/Sounds/Artemis/POI/Reality_Bloom_WildWest/Spots/Spot_Wood_Creak_Rnd_04}}
 {id=WavePlayer_7, properties {soundWaveAssetPtr=/Game/Sounds/Artemis/POI/Reality_Bloom_WildWest/Spots/Spot_Wood_Creak_Rnd_05}}
]
properties {
 attenuationSettings=/Game/Sounds/Attenuation/Fort_TileAmbient_OneShot_MorePositional_Att
 bEnableBusSends=false
 maxDistance=8000
 soundClassObject=&Fort_Ambient_Indoors_Athena
 volumeMultiplier=0.251
 soundSubmixSends [{minSendLevel=0, sendLevel=0, soundSubmix=null}]
}
